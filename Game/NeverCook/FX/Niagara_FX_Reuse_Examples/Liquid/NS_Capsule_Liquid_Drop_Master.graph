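# Niagara liquid-drop splash effect in the FX reuse examples, built from two near-identical sprite emitters, with Drops001 offset sideways.
# Each bursts camera-offset droplets in a wide cone, then applies gravity and drag.
# Sprite size pops then shrinks over lifetime, with an extra curve driven by LODDistance scaling it up with distance.
# The emitters differ in dynamic material parameters: Drops animates one by curve, while Drops001 randomizes several across two indices.
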NiagaraSystem /Game/NeverCook/FX/Niagara_FX_Reuse_Examples/Liquid/NS_Capsule_Liquid_Drop_Master
dependencies {
 usedDataInterfaces [&NiagaraDataInterfaceCurve]
 usedDynamicInputs [&MaskFloatBySpawnGroup, &Lerp_Float]
 usedModules [&SystemState]
}
emitters {
 Drops {
  emitterUpdateScript [
   &EmitterState()
   &SpawnBurst_Instantaneous(Spawn Count=45, Use Spawn Probability=false, UseLoopCountLimit=false)
  ]
  particleSpawnScript [
   &InitializeParticle(
    Lifetime Max=3
    Lifetime Min=2
    Lifetime Mode=::NewEnumerator1
    Sprite Size Mode=::NewEnumerator1
    Uniform Sprite Size Max=22
    Uniform Sprite Size Min=20
    UsePositionOffset=false
    Write Material Random=true
   )
   &AddVelocity(
    Cone Angle=140
    Module.UseConeFalloff=false
    Module.UseVelDistribution=false
    Override Seed=false
    Velocity Mode=::NewEnumerator2
    Velocity Speed=&RandomRangeFloat(Minimum=300, Maximum=900)
   )
   &CameraOffset(Camera Offset Amount=12)
  ]
  particleUpdateScript [
   &ParticleState()
   &GravityForce(Gravity=<0 0 -1600>)
   &Drag(Drag=3, Rotational Drag=3, Use Linear Drag=true, Use Rotational Drag=true) {module=Drag001}
   &SolveForcesAndVelocity(Clamp Velocity=false, Limit Acceleration=false)
   &DynamicMaterialParameters(
    Index 0 Param 1=&FloatFromCurve(
     FloatCurve=DataInterface() {propertyValues {Curve {keys [
         {arriveTangent=-5.8452, interpMode=RCIM_Cubic, time=0, value=1}
         {arriveTangentWeight=0.0434, interpMode=RCIM_Cubic, leaveTangentWeight=0.217, tangentMode=RCTM_User, time=0.2, value=0}
         {arriveTangent=3.4164, interpMode=RCIM_Cubic, leaveTangent=3.4164, tangentMode=RCTM_User, time=0.8, value=0.7}
         <at 1 0 RCIM_Cubic>
        ]}}}
     Scale Curve=1.5
    )
    Index 0 Param 2=10
    Index 0 Param 3=700
    Index 0 Param 4=5
    Module.Param0WriteEnabled=true
    Module.Param1WriteEnabled=true
    Module.Param2WriteEnabled=true
    Module.Param3WriteEnabled=true
   )
   &ScaleSpriteSize(
    Uniform Curve Scale=&FloatFromCurve(
     FloatCurve=DataInterface() {propertyValues {Curve {keys [
         <at 200 0.25 RCIM_Cubic>
         <at 1000 1>
         {arriveTangentWeight=500, interpMode=RCIM_Cubic, leaveTangentWeight=833.3334, tangentMode=RCTM_User, time=2500, value=1}
         <at 5000 1.4>
        ]}}}
     CurveIndex=$Engine.Owner.LODDistance
    )
    Uniform Curve Sprite Scale=DataInterface() {propertyValues {Curve {keys [
        {arriveTangent=40.4327, interpMode=RCIM_Cubic, leaveTangent=40.4327, tangentMode=RCTM_User, time=0, value=0.7}
        {arriveTangent=-9.8929, interpMode=RCIM_Cubic, leaveTangent=-9.8929, tangentMode=RCTM_User, time=0.07, value=1.5}
        <at 0.16 0.8>
        {interpMode=RCIM_Cubic, tangentMode=RCTM_User, time=0.75, value=0.8}
        {arriveTangent=-4.4982, interpMode=RCIM_Cubic, leaveTangent=-4.4982, tangentMode=RCTM_User, time=1, value=0}
       ]}}}
   )
  ]
  rendererClasses [&NiagaraSpriteRendererProperties]
  renderers [{rendererClass=&NiagaraSpriteRendererProperties}]
 }
 Drops001 {
  emitterUpdateScript [
   &EmitterState()
   &SpawnBurst_Instantaneous(Spawn Count=45, Use Spawn Probability=false, UseLoopCountLimit=false)
  ]
  particleSpawnScript [
   &InitializeParticle(
    Lifetime Max=3
    Lifetime Min=2
    Lifetime Mode=::NewEnumerator1
    Position Offset=<-300 300 0>
    Sprite Size Mode=::NewEnumerator1
    Uniform Sprite Size Max=22
    Uniform Sprite Size Min=20
    UsePositionOffset=true
    Write Material Random=true
   )
   &AddVelocity(
    Cone Angle=140
    Module.UseConeFalloff=false
    Module.UseVelDistribution=false
    Override Seed=false
    Velocity Mode=::NewEnumerator2
    Velocity Speed=&RandomRangeFloat(Minimum=300, Maximum=900)
   )
   &CameraOffset(Camera Offset Amount=12)
  ]
  particleUpdateScript [
   &ParticleState()
   &GravityForce(Gravity=<0 0 -1600>)
   &Drag(Drag=3, Rotational Drag=3, Use Linear Drag=true, Use Rotational Drag=true) {module=Drag001}
   &SolveForcesAndVelocity(Clamp Velocity=false, Limit Acceleration=false)
   &ScaleSpriteSize(
    Uniform Curve Scale=&FloatFromCurve(
     FloatCurve=DataInterface() {propertyValues {Curve {keys [
         <at 200 0.25 RCIM_Cubic>
         <at 1000 1>
         {arriveTangentWeight=500, interpMode=RCIM_Cubic, leaveTangentWeight=833.3334, tangentMode=RCTM_User, time=2500, value=1}
         <at 5000 1.4>
        ]}}}
     CurveIndex=$Engine.Owner.LODDistance
    )
    Uniform Curve Sprite Scale=DataInterface() {propertyValues {Curve {keys [
        {arriveTangent=40.4327, interpMode=RCIM_Cubic, leaveTangent=40.4327, tangentMode=RCTM_User, time=0, value=0.7}
        {arriveTangent=-9.8929, interpMode=RCIM_Cubic, leaveTangent=-9.8929, tangentMode=RCTM_User, time=0.07, value=1.5}
        <at 0.16 0.8>
        {interpMode=RCIM_Cubic, tangentMode=RCTM_User, time=0.75, value=0.8}
        {arriveTangent=-4.4982, interpMode=RCIM_Cubic, leaveTangent=-4.4982, tangentMode=RCTM_User, time=1, value=0}
       ]}}}
   )
   &DynamicMaterialParameters(
    Index 0 Param 1=&RandomRangeFloat(Minimum=1, Maximum=2)
    Index 0 Param 2=0.5
    Index 0 Param 3=&RandomRangeFloat(Minimum=15, Maximum=50)
    Index 0 Param 4=&RandomRangeFloat(Minimum=2, Maximum=10)
    Index 1 Param 1=&RandomRangeFloat(Minimum=0, Maximum=10)
    Module.Param0WriteEnabled=true
    Module.Param0WriteEnabled1=true
    Module.Param1WriteEnabled=true
    Module.Param1WriteEnabled1=false
    Module.Param2WriteEnabled=true
    Module.Param2WriteEnabled1=false
    Module.Param3WriteEnabled=true
    Module.Param3WriteEnabled1=false
    Write Parameter Index 1=true
   ) {module=DynamicMaterialParameters001}
  ]
  rendererClasses [&NiagaraSpriteRendererProperties]
  renderers [{rendererClass=&NiagaraSpriteRendererProperties}]
 }
}
summary {
 systemName=NS_Capsule_Liquid_Drop_Master
 emitters [
  {emitterName=Drops, rendererClasses [&NiagaraSpriteRendererProperties]}
  {emitterName=Drops001, rendererClasses [&NiagaraSpriteRendererProperties]}
 ]
 userVariables []
}
