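# Niagara system for a ribbon effect, named as an impact wave among liquid FX reuse examples.
# Its single emitter, Empty, spawns particles at a steady rate with a fixed width, finite lifetime and overbright HDR white color.
# Each particle gets a constant velocity along negative X, so the ribbon renderer draws a straight moving strip.
# It exposes no user parameters and adds no forces.
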NiagaraSystem /Game/NeverCook/FX/Niagara_FX_Reuse_Examples/Liquid/NS_Impact_Wave
dependencies {usedDataInterfaces [], usedDynamicInputs [], usedModules [&SystemState]}
emitters {Empty {
  emitterUpdateScript [
   &EmitterState()
   &SpawnRate(SpawnRate=60, Use Spawn Probability=false)
  ]
  particleSpawnScript [
   &InitializeParticle(
    Color=<rgba 10 10 10 1>
    Lifetime=3
    Ribbon Width=100
    Ribbon Width Mode=::NewEnumerator1
    UsePositionOffset=false
    Write Material Random=true
   )
   &AddVelocity(Velocity=<-100 0 0>)
  ]
  particleUpdateScript [
   &SolveForcesAndVelocity(Clamp Velocity=false, Limit Acceleration=false)
   &ParticleState()
  ]
  rendererClasses [&NiagaraRibbonRendererProperties]
  renderers [{rendererClass=&NiagaraRibbonRendererProperties}]
 }}
summary {
 systemName=NS_Impact_Wave
 emitters [{emitterName=Empty, rendererClasses [&NiagaraRibbonRendererProperties]}]
 userVariables []
}
