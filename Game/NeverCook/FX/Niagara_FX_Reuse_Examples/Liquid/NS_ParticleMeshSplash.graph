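# Niagara system for a one-shot, bright HDR blue mesh splash, apparently tied to montage swing attacks.
# A single emitter bursts one mesh particle with randomized scale and dynamic material parameters, one of them random.
# A custom dynamic input uses User.SwingSectionIndex to select the mesh's orientation vector per montage section.
# Over its life, curves grow the mesh from a tenth to full size while fading its alpha.
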NiagaraSystem /Game/NeverCook/FX/Niagara_FX_Reuse_Examples/Liquid/NS_ParticleMeshSplash
dependencies {usedDataInterfaces [&NiagaraDataInterfaceCurve], usedDynamicInputs [], usedModules [&SystemState]}
emitters {BuffSplash_Mesh {
  emitterUpdateScript [
   &EmitterState()
   &SpawnBurst_Instantaneous(Spawn Count=1, Use Spawn Probability=false)
  ]
  particleSpawnScript [
   &InitializeParticle(
    Color=<rgba 0.6574 43.5422 75 1>
    Mesh Scale Max=<10 10 6>
    Mesh Scale Min=<10 10 4>
    Mesh Scale Mode=::NewEnumerator2
    UsePositionOffset=false
    Write Material Random=true
   ) {module=InitializeParticle001}
   &DynamicMaterialParameters(
    Index 0 Param 1=0.5
    Index 0 Param 2=0.5
    Index 0 Param 3=25
    Index 0 Param 4=&RandomRangeFloat(Minimum=0, Maximum=10)
    Module.Param0WriteEnabled=true
    Module.Param1WriteEnabled=true
    Module.Param2WriteEnabled=true
    Module.Param3WriteEnabled=true
   ) {module=DynamicMaterialParameters001}
   &InitialMeshOrientation(
    Mesh Axis To Orient=<0 0 1>
    Mesh Orientation Mode=::NewEnumerator2
    Orientation Vector=/Game/Effects/Niagara/DynamicInputs/FN_DirectionalVectorfromMontageIndex(
     Montage Section Index:NiagaraInt32=$User.SwingSectionIndex
     Montage Index 1:Vector3f=$Particles.Dir1
     Montage Index 2:Vector3f=$Particles.Dir2
     Montage Index 3:Vector3f=<0 0 -1>
     Montage Index 4:Vector3f=<0 0 -1>
     Invalid Montage Index:Vector3f=<0 0 -1>
    )
    Use Rotation Vector=false
   )
  ]
  particleUpdateScript [
   &ParticleState()
   &ScaleColor(
    Scale Alpha=&FloatFromCurve(FloatCurve=DataInterface() {propertyValues {Curve {keys [
         {
          arriveTangentWeight=0.2427
          interpMode=RCIM_Cubic
          leaveTangentWeight=0.2427
          tangentMode=RCTM_User
          tangentWeightMode=RCTWM_WeightedBoth
          time=0
          value=1
         }
         {
          arriveTangent=-2.3078
          arriveTangentWeight=0.6659
          interpMode=RCIM_Cubic
          leaveTangent=-2.3078
          leaveTangentWeight=0.6659
          tangentMode=RCTM_User
          tangentWeightMode=RCTWM_WeightedBoth
          time=1
          value=0.1
         }
        ]}}})
    ScaleA=true
    ScaleRGB=false
   )
   &ScaleMeshSize(Scale Factor=&VectorFromFloat(Value=&FloatFromCurve(FloatCurve=DataInterface() {propertyValues {Curve {keys [{tangentMode=RCTM_User, time=0, value=0.1}, {tangentMode=RCTM_User, time=1, value=1}]}}}))) {module=ScaleMeshSize001}
  ]
  rendererClasses [&NiagaraMeshRendererProperties]
  renderers [{rendererClass=&NiagaraMeshRendererProperties, rendererIndex=0}]
 }}
summary {
 systemName=NS_ParticleMeshSplash
 emitters [{emitterName=BuffSplash_Mesh, rendererClasses [&NiagaraMeshRendererProperties]}]
 userVariables [User.SwingSectionIndex {type=NiagaraInt32}]
}
