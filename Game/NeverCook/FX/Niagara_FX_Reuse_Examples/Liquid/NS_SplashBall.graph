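# Niagara system for a liquid splash ball effect, with one Ball emitter that bursts three teal mesh particles from a small sphere.
# Gravity and curve-driven drag that eases off over lifetime shape their motion.
# Meshes grow sharply over their short life while a camera offset shrinks to zero.
# Dynamic material parameters, one animated by a falling curve, feed the mesh material, likely to erode the splash.
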NiagaraSystem /Game/NeverCook/FX/Niagara_FX_Reuse_Examples/Liquid/NS_SplashBall
dependencies {usedDataInterfaces [&NiagaraDataInterfaceCurve], usedDynamicInputs [], usedModules [&SystemState]}
emitters {Ball {
  emitterUpdateScript [
   &EmitterState()
   &SpawnBurst_Instantaneous(Spawn Count=3, Use Spawn Probability=false, UseLoopCountLimit=false)
  ]
  particleSpawnScript [
   &PartitionParticles(
    Normalized Index In Partition Includes One=true
    Number Of Partitions=1
    Particles Per Partition=3
    Partition Mode:ENiagara_PartitionMode=::NewEnumerator1
   )
   &InitializeParticle(
    Color=<rgba 0 0.1615 0.0898 1>
    Lifetime Max=0.8
    Lifetime Mode=::NewEnumerator1
    Mesh Scale Mode=::NewEnumerator1
    Mesh Uniform Scale Max=25
    Mesh Uniform Scale Min=20
    UsePositionOffset=false
    Write Material Random=true
   )
   &InitialMeshOrientation(Use Rotation Vector=false)
   &AddVelocity(Velocity Speed Scale=7)
   &ShapeLocation(MaskSpawnGroup=false, Override Seed=false, Sphere Radius=30, UseOwnerScale=true)
  ]
  particleUpdateScript [
   &ParticleState()
   &GravityForce(Gravity=&Multiply_VectorByFloat(Vector=<0 0 -980>, Float=1))
   &Drag(
    Drag=&FloatFromCurve(
     FloatCurve=DataInterface() {propertyValues {Curve {keys [
         {
          arriveTangent=-6.4554
          arriveTangentWeight=0.9546
          interpMode=RCIM_Cubic
          leaveTangent=-6.4554
          leaveTangentWeight=0.9546
          tangentMode=RCTM_User
          tangentWeightMode=RCTWM_WeightedBoth
          time=0
          value=1
         }
         {
          arriveTangentWeight=0.905
          interpMode=RCIM_Cubic
          leaveTangentWeight=0.905
          tangentMode=RCTM_User
          tangentWeightMode=RCTWM_WeightedBoth
          time=1
          value=0
         }
        ]}}}
     Scale Curve=7
    )
    Rotational Drag=0
    Use Linear Drag=true
    Use Rotational Drag=true
   )
   &SolveForcesAndVelocity(Clamp Velocity=false, Limit Acceleration=false)
   &DynamicMaterialParameters(
    Index 0 Param 2=0.5
    Index 0 Param 3=2
    Index 0 Param 4=&FloatFromCurve(FloatCurve=DataInterface() {propertyValues {Curve {keys [
         {arriveTangent=-3.7762, interpMode=RCIM_Cubic, leaveTangent=-3.7762, tangentMode=RCTM_User, time=0, value=1}
         {arriveTangent=-1.2694, interpMode=RCIM_Cubic, leaveTangent=-1.2694, tangentMode=RCTM_User, time=1, value=-0.9}
        ]}}})
    Index 1 Param 1=1
    Index 1 Param 2=1
    Module.Param0WriteEnabled=true
    Module.Param0WriteEnabled1=true
    Module.Param1WriteEnabled=true
    Module.Param1WriteEnabled1=true
    Module.Param2WriteEnabled=true
    Module.Param2WriteEnabled1=true
    Module.Param3WriteEnabled=true
    Module.Param3WriteEnabled1=true
    Write Parameter Index 1=true
   )
   &ScaleMeshSize(Scale Factor=&VectorFromFloat(Value=&FloatFromCurve(
      FloatCurve=DataInterface() {propertyValues {Curve {keys [
          {arriveTangent=1.0798, interpMode=RCIM_Cubic, leaveTangent=1.0798, tangentMode=RCTM_User, time=0, value=0.2}
          {arriveTangent=2.27, interpMode=RCIM_Cubic, leaveTangent=2.27, tangentMode=RCTM_User, time=1, value=5}
         ]}}}
      Scale Curve=2
     )))
   &CameraOffset(Camera Offset Amount=&FloatFromCurve(
     FloatCurve=DataInterface() {propertyValues {Curve {keys [
         {
          arriveTangentWeight=0.9309
          interpMode=RCIM_Cubic
          leaveTangentWeight=0.9309
          tangentMode=RCTM_User
          tangentWeightMode=RCTWM_WeightedBoth
          time=0
          value=1
         }
         {
          arriveTangent=-5.1083
          arriveTangentWeight=0.6812
          interpMode=RCIM_Cubic
          leaveTangent=-5.1083
          leaveTangentWeight=0.6812
          tangentMode=RCTM_User
          tangentWeightMode=RCTWM_WeightedBoth
          time=1
          value=0
         }
        ]}}}
     Scale Curve=30
    ))
  ]
  rendererClasses [&NiagaraMeshRendererProperties]
  renderers [{rendererClass=&NiagaraMeshRendererProperties, rendererIndex=0}]
 }}
summary {systemName=NS_SplashBall, emitters [{emitterName=Ball, rendererClasses [&NiagaraMeshRendererProperties]}], userVariables []}
