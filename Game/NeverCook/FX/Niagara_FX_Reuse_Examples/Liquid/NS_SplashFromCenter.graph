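# Niagara liquid splash effect that throws sprites outward and upward from a center point, from the reuse examples folder.
# Its single Splash_Hard emitter combines a burst with a brief spawn rate on a ring, using spawn-group masking to mix partition-indexed and random placement.
# Point-based velocity, gravity and drag shape the arc, while sprites grow, spin and fade late.
# Two sprite renderers draw SubUV frames, with curves animating dynamic material parameters.
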NiagaraSystem /Game/NeverCook/FX/Niagara_FX_Reuse_Examples/Liquid/NS_SplashFromCenter
dependencies {
 usedDataInterfaces [&NiagaraDataInterfaceCurve, &NiagaraDataInterfaceVector2DCurve]
 usedDynamicInputs []
 usedModules [&SystemState]
}
emitters {Splash_Hard {
  emitterUpdateScript [
   &EmitterState(Life Cycle Mode=::NewEnumerator1, Loop Behavior=::NewEnumerator1, Loop Duration=0.15, UseLoopDelay=true)
   &SpawnBurst_Instantaneous(Spawn Count=10, Use Spawn Probability=false, UseLoopCountLimit=false) {module=SpawnBurst_Instantaneous001}
   &SpawnRate(SpawnGroup=1, SpawnRate=75, Use Spawn Probability=true)
  ]
  particleSpawnScript [
   &InitializeParticle(
    Color=&RandomRangeLinearColor(Minimum=<rgba 1 1 1 1>, Maximum=<rgba 0.138 0.6337 0.1353 1>)
    Lifetime Max=1
    Lifetime Mode=::NewEnumerator1
    Position Offset Coordinate Space=::Local
    Sprite Rotation Mode=::NewEnumerator2
    Sprite Size Max=<100 150>
    Sprite Size Min=<80 200>
    Sprite Size Mode=::NewEnumerator2
    Sprite UV Mode=::NewEnumerator3
    UsePositionOffset=true
    Write Material Random=true
   )
   &PartitionParticles(Number Of Partitions=1, Particles Per Partition=7, Partition Mode:ENiagara_PartitionMode=::NewEnumerator1)
   &SubUVAnimation(End Frame=3, SubUV Animation Mode:ENiagara_SubUVLookupMode=::NewEnumerator2)
   &ShapeLocation(
    MaskSpawnGroup=false
    Override Seed=false
    Ring / Disc Distribution Mode:ENiagaraSphereDistributionMode=::NewEnumerator1
    Ring Radius=20
    Shape Primitive=::NewEnumerator4
    U Position=&MaskFloatBySpawnGroup(
     Masked Float=$Particles.PartitionParticles.Normalized Index In Partition
     Passthrough Float=&RandomRangeFloat(Minimum=0, Maximum=1)
    )
    UseOwnerScale=true
   )
   &AddVelocity(
    Module.Use Falloff Distance=false
    Origin Offset=<0 0 -50>
    Override Seed=false
    Velocity Mode=::NewEnumerator1
    Velocity Speed=&Multiply_Float(A=&RandomRangeFloat(Minimum=500, Maximum=600), B=1)
   )
   &AddVelocity(Module.Use Falloff Distance=false, Override Seed=false, Velocity Mode=::NewEnumerator1, Velocity Speed=200) {module=AddVelocity001}
  ]
  particleUpdateScript [
   &ParticleState()
   &ScaleSpriteSize(Scale Sprite Size Mode=::NewEnumerator0) {module=ScaleSpriteSize001}
   &GravityForce(Gravity=&Multiply_VectorByFloat(Vector=<0 0 -200>, Float=3)) {module=GravityForce001}
   &Drag(
    Drag=&RandomRangeFloat(Minimum=1, Maximum=1)
    Rotational Drag=50
    Use Linear Drag=true
    Use Rotational Drag=false
   )
   &SolveForcesAndVelocity(Clamp Velocity=false, Limit Acceleration=false)
   &SpriteRotationRate(Rotation Rate=&FloatFromCurve(
     FloatCurve=DataInterface() {propertyValues {Curve {keys [<at 0 1>,<at 1 0>]}}}
     Scale Curve=&MakeCustomFloatFromBool(Bool=&RandomBool(), True Float=-90, False Float=90)
    ))
   &ScaleSpriteSize(Scale Factor=&Vector2DFromCurve(
     Vector2Curve=DataInterface() {propertyValues {
       XCurve {keys [{arriveTangent=0.5278, interpMode=RCIM_Cubic, time=0, value=0.5}, {arriveTangent=1.8113, tangentMode=RCTM_User, time=0.4, value=1}]}
       YCurve {keys [
         {arriveTangent=1.4008, interpMode=RCIM_Cubic, leaveTangent=1.4008, tangentMode=RCTM_User, time=0, value=0.2}
         {interpMode=RCIM_Cubic, tangentMode=RCTM_User, time=0.75, value=1}
        ]}
      }}
     ScaleCurve=<1 1.5>
    ))
   &ScaleSpriteSizeBySpeed(Max Scale Factor=<1.2 1.5>, Velocity Threshold=800)
   &ScaleColor(
    Scale Alpha=&FloatFromCurve(FloatCurve=DataInterface() {propertyValues {Curve {keys [{interpMode=RCIM_Cubic, tangentMode=RCTM_User, time=0.9, value=1}, {interpMode=RCIM_Cubic, tangentMode=RCTM_User, time=1, value=0}]}}})
    ScaleA=true
    ScaleRGB=true
   )
   &DynamicMaterialParameters(
    Index 0 Param 2=&FloatFromCurve(
     FloatCurve=DataInterface() {propertyValues {Curve {keys [{arriveTangent=-1.9579, interpMode=RCIM_Cubic, time=0.1, value=0.2}, {interpMode=RCIM_Cubic, tangentMode=RCTM_User, time=0.3, value=0}]}}}
     Scale Curve=0.5
    )
    Index 0 Param 4=&FloatFromCurve(
     FloatCurve=DataInterface() {propertyValues {Curve {keys [
         {arriveTangent=1.2665, interpMode=RCIM_Cubic, leaveTangent=1.2665, tangentMode=RCTM_User, time=0.1, value=0}
         {arriveTangent=1.0703, interpMode=RCIM_Cubic, leaveTangent=1.0703, tangentMode=RCTM_User, time=0.75, value=2}
        ]}}}
     Scale Curve=&RandomRangeFloat(Minimum=0.5, Maximum=0.8)
    )
    Module.Param0WriteEnabled=false
    Module.Param1WriteEnabled=true
    Module.Param2WriteEnabled=false
    Module.Param3WriteEnabled=true
   )
  ]
  rendererClasses [&NiagaraSpriteRendererProperties]
  renderers [
   {rendererClass=&NiagaraSpriteRendererProperties}
   {rendererClass=&NiagaraSpriteRendererProperties, rendererIndex=1}
  ]
 }}
summary {
 systemName=NS_SplashFromCenter
 emitters [{emitterName=Splash_Hard, rendererClasses [&NiagaraSpriteRendererProperties]}]
 userVariables []
}
