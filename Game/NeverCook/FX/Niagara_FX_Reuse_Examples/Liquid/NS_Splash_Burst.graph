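# Niagara liquid splash effect: a single emitter burst-spawns short-lived, red-tinted mesh particles with a small upward velocity.
# Curves multiply into the mesh scale so the splash widens over its life while its height-to-width ratio drops from tall to flat.
# Gravity ramps in gradually as color brightness decays from an overbright start.
# A rising curve drives a dynamic material parameter, and no user parameters are exposed.
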NiagaraSystem /Game/NeverCook/FX/Niagara_FX_Reuse_Examples/Liquid/NS_Splash_Burst
dependencies {usedDataInterfaces [&NiagaraDataInterfaceCurve], usedDynamicInputs [], usedModules [&SystemState]}
emitters {Minimal {
  emitterUpdateScript [
   &EmitterState(Life Cycle Mode=::NewEnumerator1, Loop Behavior=::NewEnumerator1, Loop Duration=1, UseLoopDelay=false)
   &SpawnBurst_Instantaneous(Use Spawn Probability=false, UseLoopCountLimit=false)
  ]
  particleSpawnScript [
   &InitializeParticle(
    Color=<rgba 1 0.0599 0.0695 1>
    Lifetime=0.8
    Mesh Scale=&Multiply_VectorByFloat(Vector=<1 1 1.5>, Float=5)
    Mesh Scale Mode=::NewEnumerator4
    UsePositionOffset=false
    Write Material Random=true
   )
   &AddVelocity(Velocity=<0 0 20>)
  ]
  particleUpdateScript [
   &ParticleState()
   &ScaleMeshSize(Scale Factor=&Multiply_VectorByFloat(
     Vector=&MakeVector(
      X=1.5
      Y=1.5
      Z=&FloatFromCurve(FloatCurve=DataInterface() {propertyValues {Curve {keys [
           {tangentMode=RCTM_User, time=0, value=3}
           {arriveTangent=-2.8, interpMode=RCIM_Cubic, leaveTangent=-2.8, time=0.5, value=1.125}
           {arriveTangent=-3.7129, interpMode=RCIM_Cubic, leaveTangent=-3.7129, tangentMode=RCTM_User, time=1, value=0.2}
          ]}}})
     )
     Float=&FloatFromCurve(FloatCurve=DataInterface() {propertyValues {Curve {keys [
          {arriveTangent=1.7414, interpMode=RCIM_Cubic, leaveTangent=1.7414, tangentMode=RCTM_User, time=0.1, value=0.1}
          {arriveTangent=1.466, interpMode=RCIM_Cubic, leaveTangent=1.466, tangentMode=RCTM_User, time=1, value=1}
         ]}}})
    ))
   &GravityForce(Gravity=&Multiply_VectorByFloat(
     Vector=<0 0 -150>
     Float=&FloatFromCurve(FloatCurve=DataInterface() {propertyValues {Curve {keys [<at 0 0>,<at 0.7178 1>]}}})
    ))
   &SolveForcesAndVelocity(Clamp Velocity=false, Limit Acceleration=false)
   &ScaleColor(
    Scale RGB=&VectorFromFloat(Value=&FloatFromCurve(FloatCurve=DataInterface() {propertyValues {Curve {keys [
          <at 0 10>
          {arriveTangent=-7.1415, interpMode=RCIM_Cubic, leaveTangent=-7.1415, time=0.4058, value=5}
          {arriveTangent=-3.2451, interpMode=RCIM_Cubic, leaveTangent=-3.2451, tangentMode=RCTM_User, time=0.9802, value=3}
         ]}}}))
    ScaleA=true
    ScaleRGB=true
   )
   &DynamicMaterialParameters(
    Index 0 Param 3=&FloatFromCurve(FloatCurve=DataInterface() {propertyValues {Curve {keys [<at 0 0.2>,<at 1 1>]}}})
    Module.Index0Float0Write=true
    Module.Index0Float1Write=true
    Module.Index0Float2Write=true
    Module.Index0Float3Write=true
   )
  ]
  rendererClasses [&NiagaraMeshRendererProperties]
  renderers [{rendererClass=&NiagaraMeshRendererProperties, rendererIndex=0}]
 }}
summary {
 systemName=NS_Splash_Burst
 emitters [{emitterName=Minimal, rendererClasses [&NiagaraMeshRendererProperties]}]
 userVariables []
}
