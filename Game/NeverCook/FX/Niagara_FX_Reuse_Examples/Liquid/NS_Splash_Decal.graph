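# Niagara liquid effect that fakes a splash decal with one flat sprite in a single DustRing001 emitter, rather than a decal renderer.
# An instantaneous burst spawns one fixed-size, pale-blue particle with random rotation and a material random value.
# The sprite faces straight up, and a curve on its alpha fades it in, then out over its lifetime.
# It exposes a User.SpikeOffset vector, with no visible binding in the modules.
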NiagaraSystem /Game/NeverCook/FX/Niagara_FX_Reuse_Examples/Liquid/NS_Splash_Decal
dependencies {usedDataInterfaces [&NiagaraDataInterfaceCurve], usedDynamicInputs [], usedModules [&SystemState]}
emitters {DustRing001 {
  emitterUpdateScript [
   &EmitterState()
   &SpawnBurst_Instantaneous(Spawn Count=1, Use Spawn Probability=false)
  ]
  particleSpawnScript [&InitializeParticle(
    Color=&RandomRangeLinearColor(Minimum=<rgba 0.545 0.834 1 1>, Maximum=<rgba 0.561 0.8434 1 1>)
    Sprite Rotation Angle=&RandomRangeFloat(Minimum=0, Maximum=360)
    Sprite Rotation Mode=::NewEnumerator0
    Sprite Size Mode=::NewEnumerator1
    Uniform Sprite Size Max=150
    Uniform Sprite Size Min=150
    UsePositionOffset=true
    Write Material Random=true
   ) {module=InitializeParticle001}]
  particleUpdateScript [
   &ParticleState()
   &SpriteFacingAndAlignment(Sprite Facing=<0 0 1>, WriteAlignment=true, WriteFacing=true)
   &ScaleColor(
    Scale Alpha=&FloatFromCurve(
     FloatCurve=DataInterface() {propertyValues {Curve {keys [
         {interpMode=RCIM_Cubic, tangentMode=RCTM_User, time=0, value=-0.1}
         {interpMode=RCIM_Cubic, tangentMode=RCTM_User, time=0.5, value=1.15}
         {arriveTangent=-1.5072, interpMode=RCIM_Cubic, leaveTangent=-1.5072, tangentMode=RCTM_User, time=0.8, value=1}
         {tangentMode=RCTM_User, time=1, value=0}
        ]}}}
     Scale Curve=0.8
    )
    ScaleA=true
    ScaleRGB=true
   ) {module=ScaleColor001}
  ]
  rendererClasses [&NiagaraSpriteRendererProperties]
  renderers [{rendererClass=&NiagaraSpriteRendererProperties}]
 }}
summary {
 systemName=NS_Splash_Decal
 emitters [{emitterName=DustRing001, rendererClasses [&NiagaraSpriteRendererProperties]}]
 userVariables [User.SpikeOffset {type=Vector3f, description="", defaultValue {struct=Vector3f, value=<0 0 0>}}]
}
