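# Niagara splash effect with one mesh-particle emitter, SplashLong, that combines a continuous spawn rate with an instantaneous burst of tinted light-blue meshes.
# Each mesh spawns stretched in Z with random rotation about that axis.
# Over life, a vector curve widens it in X and Y while Z rises then collapses, forming the splash shape.
# A scaled curve animates a dynamic material parameter while alpha fades toward a low value.
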NiagaraSystem /Game/NeverCook/FX/Niagara_FX_Reuse_Examples/Liquid/NS_Splash_Long
dependencies {
 usedDataInterfaces [&NiagaraDataInterfaceVectorCurve, &NiagaraDataInterfaceCurve]
 usedDynamicInputs []
 usedModules [&SystemState]
}
emitters {SplashLong {
  emitterUpdateScript [
   &EmitterState()
   &SpawnRate(SpawnRate=10, Use Spawn Probability=false)
   &SpawnBurst_Instantaneous(Use Spawn Probability=false, UseLoopCountLimit=false)
  ]
  particleSpawnScript [
   &InitializeParticle(
    Color=<rgba 0.7031 1.4258 2 1>
    Lifetime Min=0.5
    Lifetime Mode=::NewEnumerator1
    Mesh Scale=<3 3 5>
    Mesh Scale Mode=::NewEnumerator4
    UsePositionOffset=false
    Write Material Random=true
   )
   &InitialMeshOrientation(
    Mesh Orientation Mode=::NewEnumerator2
    Rotation=&RandomRangeVector(Minimum=<0 0 0>, Maximum=<0 0 1>)
    Use Rotation Vector=true
   )
  ]
  particleUpdateScript [
   &ParticleState()
   &ScaleMeshSize(Scale Factor=&VectorFromCurve(
     VectorCurve=DataInterface() {propertyValues {
       XCurve {keys [
         {arriveTangent=1.8221, interpMode=RCIM_Cubic, leaveTangent=1.8221, tangentMode=RCTM_User, time=0, value=0}
         {arriveTangent=0.4715, interpMode=RCIM_Cubic, leaveTangent=0.4715, tangentMode=RCTM_User, time=1, value=1}
        ]}
       YCurve {keys [
         {arriveTangent=1.8221, interpMode=RCIM_Cubic, leaveTangent=1.8221, tangentMode=RCTM_User, time=0, value=0}
         {arriveTangent=0.4715, interpMode=RCIM_Cubic, leaveTangent=0.4715, tangentMode=RCTM_User, time=1, value=1}
        ]}
       ZCurve {keys [
         {arriveTangent=4.7636, interpMode=RCIM_Cubic, leaveTangent=4.7636, tangentMode=RCTM_User, time=0, value=0}
         {arriveTangentWeight=0.2848, interpMode=RCIM_Cubic, leaveTangentWeight=0.1221, tangentMode=RCTM_User, time=0.5, value=1}
         {arriveTangent=-5.505, interpMode=RCIM_Cubic, leaveTangent=-5.505, tangentMode=RCTM_User, time=1, value=0}
        ]}
      }}
     Scale Curve=&VectorFromFloat(Value=&Multiply_Float(A=1, B=1))
    ))
   &DynamicMaterialParameters(
    Index 0 Param 2=&FloatFromCurve(
     FloatCurve=DataInterface() {propertyValues {Curve {keys [<at 0 -1.5 RCIM_Cubic>, {arriveTangent=1.2375, interpMode=RCIM_Cubic, leaveTangent=1.2375, tangentMode=RCTM_User, time=1, value=1}]}}}
     Scale Curve=-20
    )
    Module.Index0Float0Write=true
    Module.Index0Float1Write=true
    Module.Index0Float2Write=false
    Module.Index0Float3Write=false
   )
   &ScaleColor(
    Scale Alpha=&FloatFromCurve(FloatCurve=DataInterface() {propertyValues {Curve {keys [
         {arriveTangent=-0.6531, interpMode=RCIM_Cubic, time=0, value=1}
         {arriveTangent=-0.8638, interpMode=RCIM_Cubic, leaveTangent=-0.8638, tangentMode=RCTM_User, time=0.8, value=0.3}
         {arriveTangent=-0.4827, interpMode=RCIM_Cubic, leaveTangent=-0.4827, tangentMode=RCTM_User, time=1, value=0.1}
        ]}}})
    ScaleA=true
    ScaleRGB=false
   )
  ]
  rendererClasses [&NiagaraMeshRendererProperties]
  renderers [{rendererClass=&NiagaraMeshRendererProperties, rendererIndex=0}]
 }}
summary {
 systemName=NS_Splash_Long
 emitters [{emitterName=SplashLong, rendererClasses [&NiagaraMeshRendererProperties]}]
 userVariables [User.Overall Scale {type=NiagaraFloat, defaultValue=1, description=""}]
}
